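# Additive unlit, two-sided Niagara mesh-particle material for the eye glows on the Bengal Basher Glow skin.
# The Sprite function supplies a world position offset, which is nudged along the camera vector by a parameter.
# Particle color feeds an emissive adjustment function.
# Opacity combines particle alpha, a sine-eased depth fade, a radial UV mask and a dynamic-parameter fade, with back faces dimmed to weaken the eye opposite the lightning.
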Material /BRCosmetics/Characters/Player/Male/Large/Bodies/M_LRG_BengalBasher_Suit/Skins/Glow/FX/Materials/M_Skin_BengalBasher_EyeGlows
graph {
 properties {
  bUsedWithNiagaraMeshParticles=true
  blendMode=BLEND_Additive
  pixelDepthOffsetMode=PDOM_Legacy
  shadingModel=MSM_Unlit
  twoSided=true
  usedShadingModels=Unlit
 }
 nodes [
  n1 = WorldPosition()
  n2 = MaterialFunctionCall!(
   Center Location=@n1.XYZ
   0-1 UVs
   XY scale=ScalarParameter() {defaultValue=6, parameterName=Sprite Scale}
   OverridePivot
   "Normals (Optional)"
  ) {materialFunction=&Sprite}
  n3 = Subtract(A=CameraPositionWS(), B=Add(A=@n1.XYZ, B=@n2.World position offset))
  Add(A=@n2.World position offset, B=Multiply(A=Normalize(VectorInput=@n3), B=ScalarParameter() {defaultValue=4, parameterName=Camera Offset}))
  n4 = ParticleColor()
  MaterialFunctionCall(Emissive=@n4.RGB) {materialFunction=&MF_FX_EmissiveAdjustments}
  n5 = Divide(A=Subtract(A=SceneDepth().R, B=PixelDepth().R), B=ScalarParameter() {defaultValue=6, parameterName=Depth Distance})
  n6 = Sine(None=Divide(A=Saturate(None=@n5)))
  n7 = TextureCoordinate()
  n8 = Constant() {r=0.5}
  n9 = Saturate(None=Multiply(A=Distance(A=@n7, B=@n8)) {constB=2})
  n10 = TwoSidedSign()
  n11 = Multiply(
   A=Power(Base=OneMinus(None=@n9), Exp=ScalarParameter() {defaultValue=1, parameterName=Mask Exponent})
   B=LinearInterpolate(Alpha=Saturate(None=@n10)) {constA=0.25}
  ) {desc=Makes eye opposite to lightning be less strong}
  n12 = DynamicParameter()
  n13 = Sine(None=Divide(A=Saturate(None=@n12.R)))
  Multiply(A=@n4.A, B=Multiply(A=Multiply(A=@n6, B=@n11), B=Power(Base=@n13)))
 ]
}
