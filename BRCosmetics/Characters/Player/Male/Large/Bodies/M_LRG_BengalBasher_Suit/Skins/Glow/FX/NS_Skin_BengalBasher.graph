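# Niagara skin effect for the Bengal Basher suit Glow skin that flashes short-lived cyan lightning meshes on the character's head and hands.
# One emitter spawns randomly timed particles, using User int arrays to pick location and mirroring, which set mesh variant, signed scale and skeletal bone/socket.
# Curves animate dynamic material parameters, from which glow-layer colors and parameters derive for five mesh renderers.
# Hidden-head and hidden-hand flags kill the matching particles.
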NiagaraSystem /BRCosmetics/Characters/Player/Male/Large/Bodies/M_LRG_BengalBasher_Suit/Skins/Glow/FX/NS_Skin_BengalBasher
dependencies {
 usedDataInterfaces [&NiagaraDataInterfaceCurve, &NiagaraDataInterfaceSkeletalMesh]
 usedDynamicInputs []
 usedModules [
  &self:SystemSpawnScript.SystemScriptSource.SystemScriptGraph.NiagaraNodeAssignment_0.SetVariables_#1
  &SystemState
 ]
}
emitters {SplineThickenMeshes {
  emitterUpdateScript [
   &EmitterState(
    Inactive Response=::NewEnumerator1
    Life Cycle Mode=::NewEnumerator1
    Loop Duration=&RandomRangeFloat(Minimum=0.05, Maximum=1.25)
    Recalculate Duration Each Loop=true
    UseLoopDelay=false
   )
   &SpawnRate(Spawn Probability=0.333, SpawnRate=3, Use Spawn Probability=true)
   &SpawnBurst_Instantaneous(Spawn Probability=0.5, Use Spawn Probability=true, UseLoopCountLimit=false)
  ]
  particleSpawnScript [
   /SetVariables(
    Particles.IsHead:NiagaraBool=&SetBoolByIntComparison(
     A=&SelectIntFromArray(
      Direct Array Index=&ReturnParticleIDAttributes()
      Direct Array Mode=::NewEnumerator1
      Int Selection Array=$User.LightningLocation
     )
     B=0
     Comparison Type=::NewEnumerator4
    )
    Particles.IsMirrored:NiagaraBool=&SetBoolByIntComparison(
     A=&SelectIntFromArray(
      Direct Array Index=&ReturnParticleIDAttributes()
      Direct Array Mode=::NewEnumerator1
      Int Selection Array=$User.LightningMirror
     )
     B=1
     Comparison Type=::NewEnumerator4
    )
   ) {isSetParameters=true}
   /SetVariables(Particles.IsHead=true, Particles.IsMirrored=true) {enabled=false, isSetParameters=true}
   &InitializeParticle(
    Color=<rgba 0 1 0.9942 1>
    Lifetime=0.4
    Mesh Scale=&MakeCustomVectorFromBool(
     Bool=$Particles.IsHead
     True Vector=&MakeCustomVectorFromBool(Bool=$Particles.IsMirrored, True Vector=<1 1 1>, False Vector=<1 1 -1>)
     False Vector=&MakeCustomVectorFromBool(Bool=$Particles.IsMirrored, True Vector=<1 1 1>, False Vector=<-1 -1 -1>)
    )
    Mesh Scale Mode=::NewEnumerator4
    UsePositionOffset=false
    Write Material Random=true
   )
   /SetVariables(
    Particles.EyeGlowsColor:LinearColor=<rgba 0.1 1 0.775 0.25>
    Particles.EyeGlowsVisibilityTag:NiagaraInt32=&MakeCustomIntFromBool(Bool=$Particles.IsHead, True INT=0, False INT=1)
    Particles.MeshIndex=&Add_Integer(
     A=&MakeCustomIntFromBool(Bool=$Particles.IsHead, True INT=0, False INT=4)
     B=&ModuloInt(A=&ReturnParticleIDAttributes(), B=4)
    )
   ) {isSetParameters=true}
   /SetVariables(Particles.MeshIndex=1) {enabled=false, isSetParameters=true}
  ]
  particleUpdateScript [
   &ParticleState()
   &SkeletalMeshLocation(
    Bone / Socket Index=&MakeCustomIntFromBool(
     Bool=$Particles.IsHead
     True INT=0
     False INT=&MakeCustomIntFromBool(Bool=$Particles.IsMirrored, True INT=1, False INT=2)
    )
    Bone Sampling Mode=::NewEnumerator4
    MaskSpawnGroup=false
    Skeletal Mesh=$System.SkeletalMesh
   )
   &DynamicMaterialParameters(
    Index 0 Param 1=&FloatFromCurve(FloatCurve=DataInterface() {propertyValues {Curve {keys [
         {interpMode=RCIM_Cubic, leaveTangent=3.6046, tangentMode=RCTM_Break, time=0, value=0.25}
         <at 0.2 0.5 RCIM_Constant>
         {interpMode=RCIM_Cubic, leaveTangent=3.4034, tangentMode=RCTM_Break, time=0.333, value=0.25}
         {arriveTangent=0.0283, interpMode=RCIM_Cubic, tangentMode=RCTM_Break, time=1, value=1}
        ]}}})
    Index 0 Param 2=&FloatFromCurve(FloatCurve=DataInterface() {propertyValues {Curve {keys [<at 0 0.5>, <at 0.2 1>, <at 0.4 1>, {arriveTangent=-4.0866, time=1, value=0}]}}})
    Index 0 Param 3=&FloatFromCurve(FloatCurve=DataInterface() {propertyValues {Curve {keys [<at 0 1 RCIM_Constant>,<at 0.4 0 RCIM_Constant>,<at 1 0>]}}}, Scale Curve=5)
    Index 0 Param 4=&FloatFromCurve(
     FloatCurve=DataInterface() {propertyValues {Curve {keys [<at 0 0 RCIM_Constant>,<at 0.2 1 RCIM_Constant>,<at 0.266 0.166 RCIM_Constant>,<at 0.333 1>]}}}
     Scale Curve=9
    )
    Index 1 Param 1=&FloatFromCurve(FloatCurve=DataInterface() {propertyValues {Curve {keys [<at 0 0.5 RCIM_Constant>, <at 0.2 0 RCIM_Constant>, <at 0.4 0.75>, {tangentMode=RCTM_User, time=1, value=2}]}}})
    Index 1 Param 2=&FloatFromCurve(
     FloatCurve=DataInterface() {propertyValues {Curve {keys [
         {interpMode=RCIM_Cubic, leaveTangent=25.2051, tangentMode=RCTM_Break, time=0, value=0}
         <at 0.2 2 RCIM_Constant>
         <at 0.266 1>
         <at 0.85 1>
         <at 1 0>
        ]}}}
     Scale Curve=&MakeCustomFloatFromBool(Bool=$Particles.IsHead, True Float=1.333, False Float=1.666)
    )
    Index 1 Param 3=&FloatFromCurve(FloatCurve=DataInterface() {propertyValues {Curve {keys [<at 0 0.25>,<at 0.5 0.25>,<at 1 2>]}}})
    Module.Index0Param0Write=true
    Module.Index0Param1Write=true
    Module.Index0Param2Write=true
    Module.Index0Param3Write=true
    Module.Index1Param0Write=true
    Module.Index1Param1Write=true
    Module.Index1Param2Write=true
    Module.Index1Param3Write=false
    Write Parameter Index 1=true
   )
   /SetVariables(
    Particles.LightningGlowColor:LinearColor=&MultiplyLinearColor(
     Linear Color=$Particles.Color
     Scale Factor=&MakeFloat4FromVectorAndFloat("Vector (RGB)"=&VectorFromFloat(Value=0.5), "Float (Alpha)"=1)
    )
    Particles.LightningGlowDynamic0:Vector4f=&MakeVector4(
     X=&MakeFloatFromVector4(Vector 4=$Particles.DynamicMaterialParameter, Channel=::NewEnumerator0)
     Y=&MakeFloatFromVector4(Vector 4=$Particles.DynamicMaterialParameter, Channel=::NewEnumerator1)
     Z=0
     W=0
    )
    Particles.LightningGlowDynamic1:Vector4f=&MakeVector4(
     X=&MakeFloatFromVector4(Vector 4=$Particles.DynamicMaterialParameter1, Channel=::NewEnumerator0)
     Y=&Multiply_Float(
      A=&MakeFloatFromVector4(Vector 4=$Particles.DynamicMaterialParameter1, Channel=::NewEnumerator1)
      B=1.5
     )
     Z=0.25
     W=0
    )
    Particles.LightningGlowSoftColor:LinearColor=&MultiplyLinearColor(
     Linear Color=$Particles.Color
     Scale Factor=&MakeFloat4FromVectorAndFloat("Vector (RGB)"=&VectorFromFloat(Value=0.166), "Float (Alpha)"=1)
    )
    Particles.LightningGlowSoftDynamic1:Vector4f=&MakeVector4(
     X=&MakeFloatFromVector4(Vector 4=$Particles.DynamicMaterialParameter, Channel=::NewEnumerator0)
     Y=&Multiply_Float(A=&MakeFloatFromVector4(Vector 4=$Particles.DynamicMaterialParameter1, Channel=::NewEnumerator1), B=3)
     Z=0
     W=0
    )
    Particles.LightningGlowSoftTopColor:LinearColor=<rgba 0.25 0.5 0.4866 0.2>
    Particles.LightningGlowSoftTopDynamic1:Vector4f=&MakeVector4(
     X=&MakeFloatFromVector4(Vector 4=$Particles.DynamicMaterialParameter, Channel=::NewEnumerator0)
     Y=&Multiply_Float(A=&MakeFloatFromVector4(Vector 4=$Particles.DynamicMaterialParameter1, Channel=::NewEnumerator1), B=3)
     Z=0
     W=0
    )
   ) {isSetParameters=true}
   &KillParticles(Kill Particles=&Boolean_OrOperation(
     Bool1=&Boolean_AndOperation(Bool1=$User.bisHeadHidden, Bool2=$Particles.IsHead)
     Bool2=&Boolean_AndOperation(Bool1=$User.bisHandHidden, Bool2=&Boolean_NotOperation(Bool=$Particles.IsHead))
    ))
  ]
  rendererClasses [&NiagaraMeshRendererProperties]
  renderers [
   {rendererClass=&NiagaraMeshRendererProperties, rendererIndex=0}
   {rendererClass=&NiagaraMeshRendererProperties, rendererIndex=1}
   {rendererClass=&NiagaraMeshRendererProperties, rendererIndex=2}
   {rendererClass=&NiagaraMeshRendererProperties, rendererIndex=3}
   {rendererClass=&NiagaraMeshRendererProperties, rendererIndex=4}
  ]
 }}
summary {
 systemName=NS_Skin_BengalBasher
 emitters [{emitterName=SplineThickenMeshes, rendererClasses [&NiagaraMeshRendererProperties]}]
 userVariables [
  User.bIsFrontEndPreview {type=NiagaraBool}
  User.bisHandHidden {type=NiagaraBool}
  User.bisHeadHidden {type=NiagaraBool}
  User.bisInVehicle {type=NiagaraBool}
  User.LightningLocation {
   type=NiagaraDataInterfaceArrayInt32
   description=""
   defaultValue {
    struct=NiagaraExt_VariableValue_DataInterface
    value {
     dataInterface=&self:NiagaraDataInterfaceArrayInt32_2
     dataInterfaceClass=&NiagaraDataInterfaceArrayInt32
    }
   }
  }
  User.LightningMirror {
   type=NiagaraDataInterfaceArrayInt32
   description=""
   defaultValue {
    struct=NiagaraExt_VariableValue_DataInterface
    value {
     dataInterface=&self:NiagaraDataInterfaceArrayInt32_1
     dataInterfaceClass=&NiagaraDataInterfaceArrayInt32
    }
   }
  }
 ]
}
